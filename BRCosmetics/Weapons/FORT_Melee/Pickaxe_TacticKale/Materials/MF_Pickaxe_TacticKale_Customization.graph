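# Material function for the TacticKale pickaxe cosmetic that recolors the Base Color of incoming material attributes for customization.
# Channel-mask parameters split a TAMask input into three zones, each with a color parameter that a TestColors static switch can swap for a test color.
# The M input supplies a skin mask and an overlay blend offset by a scalar.
# Nested lerps and multiplies build the new Base Color, which is written back into the attributes.
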MaterialFunction /BRCosmetics/Weapons/FORT_Melee/Pickaxe_TacticKale/Materials/MF_Pickaxe_TacticKale_Customization
graph {nodes [
  n1 = LinearInterpolate(A=NamedRerouteUsage(), B=Multiply(A=NamedRerouteUsage(), B=NamedRerouteUsage()), Alpha=NamedRerouteUsage())
  n2 = Multiply(A=NamedRerouteUsage(), B=NamedRerouteUsage())
  n3 = NamedRerouteUsage()
  FunctionOutput(None=SetMaterialAttributes!(MaterialAttributes=NamedRerouteUsage(), BaseColor=LinearInterpolate(A=@n1, B=@n2, Alpha=@n3)) {attributeSetTypes [#1]})
  n4 = GetMaterialAttributes(None=FunctionInput() {inputType=FunctionInput_MaterialAttributes}) {attributeGetTypes [#1]}
  NamedRerouteDeclaration(None=@n4.MaterialAttributes) {name=MA, nodeColor=<rgba 0 1 0.9531 1>}
  NamedRerouteDeclaration(None=@n4.BaseColor) {name=BaseColor, nodeColor=<rgba 0.2188 0 1 1>}
  n5 = MaterialFunctionCall(Float3=FunctionInput() {inputName=M, sortPriority=1}) {materialFunction=&BreakOutFloat3Components}
  NamedRerouteDeclaration(None=@n5.B) {name=MSkin, nodeColor=<rgba 0.1484 1 0 1>}
  n6 = FunctionInput() {inputName=TAMask, sortPriority=2}
  NamedRerouteDeclaration(None=ChannelMaskParameter(None=@n6) {group=TacticCustomize, parameterName=Zone1}.R) {name=Zone1Mask, nodeColor=<rgba 0 1 0.6484 1>}
  NamedRerouteDeclaration(None=ChannelMaskParameter(None=@n6) {group=TacticCustomize, maskChannel=Blue, parameterName=Zone2, defaultValue=<rgba 0 0 1 0>}.R) {name=Zone2Mask, nodeColor=<rgba 0.4062 0 1 1>}
  NamedRerouteDeclaration(None=ChannelMaskParameter(None=@n6) {group=TacticCustomize, maskChannel=Green, parameterName=Zone3, defaultValue=<rgba 0 1 0 0>}.R) {name=Zone3Mask, nodeColor=<rgba 1 0 0.2344 1>}
  NamedRerouteDeclaration(None=Add(A=@n5.G, B=ScalarParameter() {group=TacticCustomize, parameterName=OverlayAdding, sortPriority=1})) {name=OverlayBlend, nodeColor=<rgba 1 0 0.7031 1>}
  n7 = VectorParameter() {group=TacticCustomize, parameterName=Test_Zone1Color, sortPriority=100, defaultValue=<rgba 0 0 0 1>}
  n8 = StaticBoolParameter() {group=TacticCustomize, parameterName=TestColors, sortPriority=99}
  NamedRerouteDeclaration(None=StaticSwitch(
    True=@n7.RGB
    False=VectorParameter() {group=TacticCustomize, parameterName=Zone1Color, sortPriority=10, defaultValue=<rgba 0 0 0 1>}.RGB
    Value=@n8
   )) {name=Zone1Color, nodeColor=<rgba 1 0 0.0446 1>}
  n9 = VectorParameter() {group=TacticCustomize, parameterName=Test_Zone2Color, sortPriority=100, defaultValue=<rgba 0 0 0 1>}
  NamedRerouteDeclaration(None=StaticSwitch(
    True=@n9.RGB
    False=VectorParameter() {group=TacticCustomize, parameterName=Zone2Color, sortPriority=11, defaultValue=<rgba 0 0 0 1>}.RGB
    Value=@n8
   )) {name=Zone2Color, nodeColor=<rgba 0 1 0 1>}
  n10 = VectorParameter() {group=TacticCustomize, parameterName=Test_Zone3Color, sortPriority=100, defaultValue=<rgba 0 0 0 1>}
  NamedRerouteDeclaration(None=StaticSwitch(
    True=@n10.RGB
    False=VectorParameter() {group=TacticCustomize, parameterName=Zone3Color, sortPriority=12, defaultValue=<rgba 0 0 0 1>}.RGB
    Value=@n8
   )) {name=Zone3Color, nodeColor=<rgba 0 0 1 1>}
 ]}
